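# Material function for the Season 16 VinderTech umbrella that sets an animated, tinted emissive on incoming material attributes.
# Time steps through a flipbook atlas of grayscale distance-field frames, with frame indexing assuming a square grid.
# It blends consecutive frames by fractional time so the shapes morph, then offsets and steeply scales the result for crisp edges.
# A color and multiplier tint that mask, which replaces the input emissive.
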MaterialFunction /BRCosmetics/Gadgets/Assets/VinderTech_Umbrella/Umbrella_Season16/Materials/MF_Umbrella_Season16
graph {nodes [
  n1 = FunctionInput() {inputType=FunctionInput_MaterialAttributes}
  n2 = Multiply(A=Time(), B=ScalarParameter() {defaultValue=2, group=TurtleNeck, parameterName=TN_Frequency, sortPriority=3})
  n3 = Frac(None=@n2)
  n4 = VectorParameter() {group=TurtleNeck, parameterName="TN_SubImages(V2)", sortPriority=2, defaultValue=<rgba 2 2 0 1>}
  n5 = MaterialFunctionCall(Float2=@n4.RGB) {materialFunction=&BreakOutFloat2Components}
  n6 = Fmod(A=Subtract(A=@n2, B=@n3), B=Multiply(A=@n5.R, B=@n5.R))
  n7 = AppendVector(A=Fmod(A=@n6, B=@n5.R), B=Floor(None=Divide(A=@n6, B=@n5.R)))
  n8 = TextureCoordinate()
  n9 = ComponentMask(None=@n4.RGB) {g=true, r=true}
  n10 = TextureSampleParameter2D(UVs=Divide(A=Add(A=@n7, B=@n8), B=@n9) {constB=3}) {
   group=TurtleNeck
   parameterName=DF Texture
   samplerType=SAMPLERTYPE_Grayscale
   sortPriority=2
   texture=/BRCosmetics/Gadgets/Assets/VinderTech_Umbrella/Umbrella_Season16/Textures/T_Umbrella_Season_16_DF
  }
  n11 = Frac(None=@n2)
  n12 = MaterialFunctionCall(Float2=@n4.RGB) {materialFunction=&BreakOutFloat2Components}
  n13 = Fmod(A=Add(A=Subtract(A=@n2, B=@n11)), B=Multiply(A=@n12.R, B=@n12.R))
  n14 = AppendVector(A=Fmod(A=@n13, B=@n12.R), B=Floor(None=Divide(A=@n13, B=@n12.R)))
  n15 = TextureSampleParameter2D(UVs=Divide(A=Add(A=@n8, B=@n14), B=@n9) {constB=3}) {
   group=TurtleNeck
   parameterName=DF Texture
   samplerType=SAMPLERTYPE_Grayscale
   sortPriority=2
   texture=&T_Umbrella_Season_16_DF
  }
  n16 = Saturate(None=Add(
    A=LinearInterpolate(A=@n10.RGB, B=@n15.RGB, Alpha=@n3)
    B=ScalarParameter() {defaultValue=-0.4, group=TurtleNeck, parameterName=TN_SurfaceDistanceOffset, sortPriority=11}
   ) {constB=-0.25})
  n17 = Constant() {r=14}
  n18 = Multiply(A=Saturate(None=Multiply(A=@n16, B=@n17)), B=VectorParameter() {parameterName=TN_EmsColor, defaultValue=<rgba 0.7552 0.0149 0 1>}.RGB)
  n19 = ScalarParameter() {defaultValue=2, group=TurtleNeck, parameterName=TN_EmsMult}
  FunctionOutput(None=SetMaterialAttributes!(MaterialAttributes=GetMaterialAttributes(None=@n1).MaterialAttributes, EmissiveColor=Multiply(A=@n18, B=@n19)) {attributeSetTypes [#1]})
 ]}
