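# Master material for the Season 16 VinderTech umbrella cosmetic, built on material attributes for skinned meshes.
# Weapon master output enters wrap customization, with bounds and safe-zone parameters also fed through a zero-alpha Base Color lerp.
# Elemental and overheat functions follow, bypassed on mobile, then ADS scope.
# A static switch optionally drives World Position Offset to hide vertex-colored elements, before umbrella-specific and accessory tech-art override functions.
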Material /BRCosmetics/Gadgets/Assets/VinderTech_Umbrella/Umbrella_Season16/Materials/MAT_Umbrella_Season16_Master
graph {
 properties {
  bUseMaterialAttributes=true
  bUsedWithMeshDeformer=true
  bUsedWithSkeletalMesh=true
  pixelDepthOffsetMode=PDOM_Legacy
  refractionMethod=RM_IndexOfRefraction
  usedShadingModels=Default Lit
 }
 nodes [
  Constant()
  n1 = MaterialFunctionCall!(In) {materialFunction=/Game/Packages/Fortress_SharedMaterials/Base_Material_Functions/BaseCustomizationFunctions/MF_FN_Weapon_MASTER}
  n2 = GetMaterialAttributes(None=@n1.Material Result) {attributeGetTypes [#1]}
  n3 = VectorParameter() {group=Weapon Specific Wrap Adjustments, parameterName=DO NOT OVERRIDE-OBMAX, defaultValue=<rgba 0 0 0 1>}
  n4 = VectorParameter() {group=Weapon Specific Wrap Adjustments, parameterName=DO NOT OVERRIDE-OBMIN, defaultValue=<rgba 0 0 0 1>}
  n5 = VectorParameter() {group=Weapon Specific Wrap Adjustments, parameterName=DO NOT OVERRIDE-OBCENTER, defaultValue=<rgba 0 0 0 1>}
  n6 = VectorParameter() {group=Weapon Specific Wrap Adjustments, parameterName=DO NOT OVERRIDE-SZ-Back, sortPriority=100, defaultValue=<rgba 0.3 0 0 0.2>}
  n7 = VectorParameter() {group=Weapon Specific Wrap Adjustments, parameterName=DO NOT OVERRIDE-SZ-Middle, sortPriority=101, defaultValue=<rgba 0 0 0 0.2>}
  n8 = VectorParameter() {group=Weapon Specific Wrap Adjustments, parameterName=DO NOT OVERRIDE-SZ-Front, sortPriority=102, defaultValue=<rgba -0.3 0 0 0.2>}
  n9 = Add(A=Add(A=Add(A=@n3.RGB, B=@n4.RGB), B=@n5.RGB), B=Add(A=Add(A=@n6.RGB, B=@n7.RGB), B=@n8.RGB))
  n10 = MaterialFunctionCall!(
   Material=SetMaterialAttributes!(MaterialAttributes=@n2.MaterialAttributes, Base Color=LinearInterpolate(A=@n2.Base Color, B=@n9) {constAlpha=0}) {attributeSetTypes [#1]}
   ------------------
   Object Bounds Center Adjustment=@n5.RGB
   Object Bounds Min Adjustment=@n4.RGB
   Wrap ScaleMultiplier
   Rotation X Axis
   "Customization Mask(_CM A)"=@n1.Customization Gmap
   Forward Axis=VectorParameter() {parameterName=WeaponForwardAxis, defaultValue=<rgba 0 1 0 1>}.RGB
   Object Bounds Max Adjustment=@n3.RGB
   Rotation Z Axis
   Normal Map=@n1.Customization Gmap
   Cropped Emissive Mask=@n1.Customization Gmap
   "Customization Crevasse (_CM G)"=@n1.Customization Gmap
   "Customization Mask UV Islands (_CM B Bottom Right)"=@n1.Customization Gmap
   -------------------
   "Customization Mask 1 (_CM B Top Left)"=@n1.Customization Gmap
   "Customization Mask 2 (_CM B Top Right)"=@n1.Customization Gmap
   "Customization Mask 3 (_CM B Bottom Left)"=@n1.Customization Gmap
   "Customization Mask (_CM RGBA)"
   --------------------
   Pre-Skinned Local Position=@n1.Customization Gmap
   Pre-Skinned Bounds=@n1.Customization Gmap
   Pre-Skinned Local Normal=@n1.Customization Gmap
   Per_ObjectMaskOverride
   isVehicle
   Side Project Vehicle
   isMech
   isCharacter
   isCharAccent
   isPickaxe
   isBackpack
   Object-SafeZone-Back=AppendVector(A=@n6.RGB, B=@n6.RGB)
   Object-SafeZone-Middle=AppendVector(A=@n7.RGB, B=@n7.RGB)
   Object-SafeZone-Front=AppendVector(A=@n8.RGB, B=@n8.RGB)
   Object-SafeZone-VP_Charcter_Preview_Switch
   Object-SafeZone-Virtual-Pivot
   Wrap Holdout
  ) {materialFunction=/Game/Packages/Fortress_SharedMaterials/Base_Material_Functions/BaseCustomizationFunctions/MF_Base_Customization}
  n11 = MaterialFunctionCall!(
   Base MA=MaterialFunctionCall!(
    In=@n10.Result
    Pre-Skinned Local Position=@n1.Pre-Skinned Local Position Pixel Shader
    Pre-Skinned Local Normal=@n1.Pre-Skinned Local Normal Pixel Shader
   ) {materialFunction=/Game/Weapons/WeaponSkins/MaterialFunction/ElementalWeaponsFunction}.Result
   Pre-Skinned Local Position=@n1.Pre-Skinned Local Position Pixel Shader
  ) {materialFunction=/Game/Weapons/WeaponSkins/MaterialFunction/OverheatGunFunction}
  n12 = GetMaterialAttributes(None=MaterialFunctionCall!(Material Attributes=ShadingPathSwitch(Default=@n11.Result, Mobile=@n10.Result) {desc=Migrated from FeatureLevelSwitch}) {materialFunction=/Game/Packages/Fortnite_UniqueMaterials_1/Weapons/ADSScopeMaster}.Result) {attributeGetTypes [#2]}
  n13 = StaticSwitchParameter(
   True=MaterialFunctionCall!(WPO) {materialFunction=/Game/Packages/Fortress_SharedMaterials/Base_Material_Functions/Fortnite_HideElementsWithVertexColor}.Result
   False=@n12.World Position Offset
  ) {group=Opacity Mask, parameterName=Use Vertex Colors for Mask}
  MaterialFunctionCall(In=MaterialFunctionCall!(In=SetMaterialAttributes!(MaterialAttributes=@n12.MaterialAttributes, World Position Offset=@n13) {attributeSetTypes [#2]}) {materialFunction=&/MF_Umbrella_Season16}.Result) {materialFunction=&MF_TechArtOverrides_accessories}
  Reroute(None=@n1.Material Result)
  SetMaterialAttributes!(MaterialAttributes, Base Color=@n1.Diffuse) {attributeSetTypes [#1]}
  TextureSample() {texture=/BRCosmetics/Weapons/FORT_Melee/Pickaxe_NeonGlow/Textures/Neon_Glow_Pickaxe_D}
 ]
}
